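; Debug actor Blueprint from the Water plugin that shows water zone texture data on a static mesh.
; The construction script creates a dynamic instance of the Deform_To_Terrain debug material and applies it to slot 0 of the mesh.
; A Get Water Texture custom event finds a WaterZone actor and feeds its water info texture array into the material's WaterVelocityAndHeight parameter.
; BeginPlay, overlap and Tick events are empty.
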(event EventBeginPlay)

(event Collision|EventActorBeginOverlap
  (params (OtherActor :type Actor%20Object%20Reference)))

(event EventTick
  (params (DeltaSeconds :type Float%20%28single-precision%29)))

(event Custom|Get%20Water%20Texture
  (bind _returnvalue (Actor|GetActorOfClass "/Script/Water.WaterZone"))
  (Rendering|Material|SetTextureParameterValue (Variables|Default|GetMID) "WaterVelocityAndHeight" (Variables|/Script/Water.WaterZone|GetWaterInfoTextureArray _returnvalue)))

(fn ConstructionScript
  (bind _mid (Variables|Default|GetMID))
  (Rendering|CreateAndAssignMID _mid "/Water/Materials/Debug/Deform_To_Terrain.Deform_To_Terrain"
    (:Out
      (Rendering|Material|SetMaterial (Variables|Default|GetStaticMesh) 0 _mid))))
